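# Physics asset for the male medium robot hologram NPC body, giving its skeleton a simplified collision and constraint setup.
# One capsule per body covers pelvis, spine, neck and limbs, linked outward from the pelvis with uniform swing and twist limit angles.
# The right side is asymmetric: lowerarm_r and thigh_r have no bodies, so hand_r joins upperarm_r, while calf_r and thigh_twist_01_r each join the pelvis directly.
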PhysicsAsset /BRCosmetics/Characters/Player/Male/Medium/Bodies/M_MED_Robot_Hologram_NPC/Meshes/M_MED_Robot_Hologram_NPC_PhysicsAsset
bodies {
 calf_l [{shapeType=Capsule, length=20.8868, radius=14.0645, center=<-18.4236 -2.835 -0.2881>, rotation=<rot 0 -90.9636 -87.7537>}]
 calf_r [{shapeType=Capsule, length=20.8868, radius=14.0641, center=<18.424 2.8354 0.2884>, rotation=<rot 0 -90.9603 -87.7584>}]
 foot_l [{shapeType=Capsule, length=14.6584, radius=7.6524, center=<-2.1023 5.7769 -0.1324>, rotation=<rot 0 -179.3319 -88.2324>}]
 foot_r [{shapeType=Capsule, length=14.6584, radius=7.6524, center=<2.1023 -5.7769 0.1324>, rotation=<rot 0 -179.3319 -88.2324>}]
 hand_l [{shapeType=Capsule, length=8.7117, radius=6.3228, center=<8.148 -2.2493 0.3466>, rotation=<rot 0 80.3496 -77.5663>}]
 hand_r [{shapeType=Capsule, length=8.7117, radius=6.3228, center=<-8.148 2.2493 -0.3466>, rotation=<rot 0 80.3496 -77.5662>}]
 lowerarm_l [{shapeType=Capsule, length=13.2732, radius=7.723, center=<10.6919 1.416 -0.7741>, rotation=<rot 0 101.4749 -86.6561>}]
 neck_02 [{shapeType=Capsule, length=14.9851, radius=11.155, center=<8.7734 -0.8532 -0.6413>, rotation=<rot 5.447 -90 -22.3435>}]
 pelvis [{shapeType=Capsule, length=32.3478, radius=17.1513, center=<-16.9278 1.6927 6.1657>, rotation=<rot 0 -101.456 -73.9584>}]
 spine_01 [{shapeType=Capsule, length=11.8551, radius=10.2883, center=<9.0974 -1.3395 0>, rotation=<rot 0 -90 0>}]
 spine_03 [{shapeType=Capsule, length=14.9629, radius=14.7569, center=<3.764 0.2938 -0.0017>, rotation=<rot 0.0898 -90 -0.0032>}]
 spine_04 [{shapeType=Capsule, length=16.7808, radius=15.8736, center=<3.4492 0.4439 0>, rotation=<rot 90 95.9793 -174.0207>}]
 spine_05 [{shapeType=Capsule, length=22.3018, radius=17.0762, center=<9.1454 -1.0789 0.0156>, rotation=<rot 0.5225 -90 0.0386>}]
 thigh_l [{shapeType=Capsule, length=23.9897, radius=11.5484, center=<-21.9439 1.5887 -0.3384>, rotation=<rot 0 -89.7992 -81.5991>}]
 thigh_twist_01_r [{shapeType=Capsule, length=15.1132, radius=13.4071, center=<-8.179 -2.5808 2.0478>, rotation=<rot 0 -28.7469 -69.636>}]
 upperarm_l [{shapeType=Capsule, length=19.1438, radius=10.3439, center=<13.158 -1.9467 -0.1948>, rotation=<rot 0 -102.7314 -82.5058>}]
 upperarm_r [{shapeType=Capsule, length=26.5295, radius=10.7193, center=<-22.7481 -1.3468 -0.0802>, rotation=<rot 0 -83.8122 -87.0964>}]
}
constraints [
 {bone1Name=spine_01, bone2Name=pelvis, swing1LimitDegrees=45, swing2LimitDegrees=45, twistLimitDegrees=45}
 {bone1Name=spine_03, bone2Name=spine_01, swing1LimitDegrees=45, swing2LimitDegrees=45, twistLimitDegrees=45}
 {bone1Name=spine_04, bone2Name=spine_03, swing1LimitDegrees=45, swing2LimitDegrees=45, twistLimitDegrees=45}
 {bone1Name=spine_05, bone2Name=spine_04, swing1LimitDegrees=45, swing2LimitDegrees=45, twistLimitDegrees=45}
 {bone1Name=neck_02, bone2Name=spine_05, swing1LimitDegrees=45, swing2LimitDegrees=45, twistLimitDegrees=45}
 {bone1Name=upperarm_l, bone2Name=spine_05, swing1LimitDegrees=45, swing2LimitDegrees=45, twistLimitDegrees=45}
 {bone1Name=lowerarm_l, bone2Name=upperarm_l, swing1LimitDegrees=45, swing2LimitDegrees=45, twistLimitDegrees=45}
 {bone1Name=hand_l, bone2Name=lowerarm_l, swing1LimitDegrees=45, swing2LimitDegrees=45, twistLimitDegrees=45}
 {bone1Name=upperarm_r, bone2Name=spine_05, swing1LimitDegrees=45, swing2LimitDegrees=45, twistLimitDegrees=45}
 {bone1Name=hand_r, bone2Name=upperarm_r, swing1LimitDegrees=45, swing2LimitDegrees=45, twistLimitDegrees=45}
 {bone1Name=thigh_l, bone2Name=pelvis, swing1LimitDegrees=45, swing2LimitDegrees=45, twistLimitDegrees=45}
 {bone1Name=calf_l, bone2Name=thigh_l, swing1LimitDegrees=45, swing2LimitDegrees=45, twistLimitDegrees=45}
 {bone1Name=foot_l, bone2Name=calf_l, swing1LimitDegrees=45, swing2LimitDegrees=45, twistLimitDegrees=45}
 {bone1Name=calf_r, bone2Name=pelvis, swing1LimitDegrees=45, swing2LimitDegrees=45, twistLimitDegrees=45}
 {bone1Name=foot_r, bone2Name=calf_r, swing1LimitDegrees=45, swing2LimitDegrees=45, twistLimitDegrees=45}
 {bone1Name=thigh_twist_01_r, bone2Name=pelvis, swing1LimitDegrees=45, swing2LimitDegrees=45, twistLimitDegrees=45}
]
